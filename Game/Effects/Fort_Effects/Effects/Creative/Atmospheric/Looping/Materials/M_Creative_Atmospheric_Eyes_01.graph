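# Additive unlit sprite material for a Creative atmospheric looping effect, drawing a pair of blinking, flickering eyes tinted by particle color.
# A sphere mask mirrors U into two eyes, then random-exponent warping and pupil-driven spherical distortion offset the eye texture lookup.
# Dynamic parameters drive variation, pupil, exponent and blink, where a moving sphere mask forms the lids.
# Two summed sines flicker brightness and lids before depth fade and emissive adjustment.
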Material /Game/Effects/Fort_Effects/Effects/Creative/Atmospheric/Looping/Materials/M_Creative_Atmospheric_Eyes_01
graph {
 properties {
  bEnableResponsiveAA=true
  bUsedWithParticleSprites=true
  blendMode=BLEND_Additive
  pixelDepthOffsetMode=PDOM_Legacy
  shadingModel=MSM_Unlit
  usedShadingModels=Unlit
 }
 nodes [
  ConstantBiasScale(None=Sine(None=Time()))
  n1 = ParticleColor()
  n2 = MaterialFunctionCall(Float2=TextureCoordinate() {uTiling=2}) {materialFunction=&BreakOutFloat2Components}
  n3 = Constant() {r=1}
  n4 = SphereMask(A=@n2.R, B=@n3, Radius=@n3, Hardness=Constant())
  n5 = Power(
   Base=AppendVector(A=OneMinus(None=@n4), B=@n2.G)
   Exp=DynamicParameter() {defaultValue=<rgba 1 1 1 0>, paramNames [EyesOpenAnimation, EyesRandom, EyeExponent, Pupil]}.EyesRandom
  )
  n6 = Constant() {r=0.5}
  n7 = SphereMask(A=@n5, B=@n6, Radius=@n6, Hardness=Constant())
  n8 = Multiply(A=@n7, B=ConstantBiasScale(None=@n5) {bias=-0.5, scale=2})
  n9 = Multiply(A=DynamicParameter() {defaultValue=<rgba 1 1 1 0>, paramNames [EyesOpenAnimation, EyesRandom, EyeExponent, Pupil]}.Pupil, B=@n8)
  n10 = ScalarParameter() {defaultValue=-0.02, parameterName=BiasU}
  n11 = ScalarParameter() {defaultValue=0.02, parameterName=BiasV}
  n12 = MaterialFunctionCall(
   UVs=Add(A=@n5, B=AppendVector(A=@n10, B=@n11))
   Texture Scale=AppendVector(A=ScalarParameter() {defaultValue=1.45, parameterName=ScaleU}, B=ScalarParameter() {defaultValue=1.45, parameterName=ScaleV})
  ) {materialFunction=&ScaleUVsByCenter}
  n13 = Power(
   Base=TextureSample(UVs=Add(A=@n9, B=@n12.UVs)) {samplerType=SAMPLERTYPE_LinearColor, texture=/Game/Packages/Fortnite_UniqueMaterials_1/Textures/New_eye_Diffuse2}.G
   Exp=DynamicParameter() {defaultValue=<rgba 1 1 1 0>, paramNames [EyesOpenAnimation, EyesRandom, EyeExponent, Pupil]}.EyeExponent
  )
  n14 = Multiply(A=DynamicParameter() {defaultValue=<rgba 1 1 1 0>, paramNames [EyesOpenAnimation, EyesRandom, EyeExponent, Pupil]}.EyesRandom, B=Time())
  n15 = Add(A=Sine(None=@n14) {period=0.2407}, B=Sine(None=@n14) {period=0.3638})
  n16 = Multiply(A=Multiply(A=@n13, B=@n15), B=ScalarParameter() {defaultValue=0.4, parameterName=EyeFlickerScale})
  n17 = Constant() {r=0.5}
  n18 = Constant() {r=0.8}
  n19 = VectorParameter() {parameterName=VectorA_BottomLid, defaultValue=<rgba 0.9 0.9 0 0>}
  n20 = Normalize(VectorInput=VectorParameter() {parameterName=VectorB_BottomLid, defaultValue=<rgba 0.7112 0.75 0.6628 0>}.RGB)
  n21 = Add(
   A=DynamicParameter() {defaultValue=<rgba 1 1 1 0>, paramNames [EyesOpenAnimation, EyesRandom, EyeExponent, Pupil]}.EyesOpenAnimation
   B=Multiply(A=@n15, B=ScalarParameter() {defaultValue=0.03, parameterName=BlinkFlickerScale})
  )
  n22 = ScalarParameter() {defaultValue=0.7, parameterName=BlinkHardness}
  n23 = Multiply(
   A=Multiply(A=Add(A=@n13, B=@n16), B=SphereMask(A=@n5, B=@n17, Radius=@n17, Hardness=@n18))
   B=SphereMask(
    A=AppendVector(A=@n8, B=@n7)
    B=LinearInterpolate(A=@n19.RGB, B=@n20, Alpha=@n21)
    Radius=ScalarParameter() {defaultValue=1, parameterName=Blink Radius}
    Hardness=LinearInterpolate(B=@n22, Alpha=@n21) {constA=1}
   ) {desc=Don't blink . . .}
  )
  MaterialFunctionCall(Emissive=DepthFade(Opacity=Multiply(A=@n1.RGB, B=@n23), FadeDistance=ScalarParameter() {defaultValue=5, parameterName=FadeDistance})) {materialFunction=&MF_FX_EmissiveAdjustments}
  ParticleColor()
 ]
}
